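# Widget for one kill feed entry in the Athena HUD: a transparent Border wrapping a CommonRichTextBlock with image and texture decorators.
# When a list item is set, it casts the object to HUDKillFeedEntryViewModel and assigns it as the view model.
# Construct sets left or right text justification from the BattleRoyalGameUI.KillFeedPositionOnRight.Enabled console variable.
# PreConstruct swaps to the Bacchus text style table when the compact (mobile) UI is in use.
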WidgetBlueprint /Game/Athena/HUD/HUDKillFeedEntryWidget
graphs [
 EventGraph {nodes [
   Event Construct()
   > n1 = GetConsoleVariableBoolValue!(execute, VariableName:String=BattleRoyalGameUI.KillFeedPositionOnRight.Enabled) -> (then, ReturnValue:bool)
   > Set Justification!(
    execute
    Justification:ETextJustify Enum=Select(Option 0:ETextJustify Enum=Left, Option 1:ETextJustify Enum=Right, Index=@n1) -> (ReturnValue:ETextJustify Enum)
    self:Text Layout Widget Object Reference=Get KillFeedEntry_Text() -> (KillFeedEntry_Text:Common Rich Text Block Object Reference)
   ) -> (then, Output_Get:ETextJustify Enum=Left)
   n2 = Event OnListItemObjectSet()
   > n3 = Cast To HUDKillFeedEntryViewModel!(execute, Object=@n2.ListItemObject) -> (then, CastFailed, AsHUD - Kill Feed Entry Viewmodel:HUD - Kill Feed Entry Viewmodel Object Reference) {is=title K2Node_DynamicCast}
   > SetHUDKillFeedEntryViewModel!(execute, self, Viewmodel:HUD - Kill Feed Entry Viewmodel Object Reference=@n3) -> (then)
   Event PreConstruct()
   > SetMobileStyle() {is=title K2Node_CallFunction}
  ]}
 SetMobileStyle {nodes [
   SetMobileStyle() {is=title K2Node_FunctionEntry}
   > Branch(Condition=IsUsingCompactUI())
   > SetTextStyleSet!(
    execute
    self:Rich Text Block Object Reference=Get KillFeedEntry_Text()
    NewTextStyleSet:Data Table Object Reference=/Game/Athena/HUD/Bacchus/BacchusKillFeedStyles
   ) -> (then)
  ]}
]
variables []
widgets {
 info {inheritedWidgetCount=0, parentClass=&UserWidget, rootWidgetClass=&Border, widgetCount=2}
 widgets [
  Border_0 {
   widgetClassPath=&Border
   widget=Border_0
   properties {
    background {
     bIsDynamicallyLoaded=false
     drawAs=Image
     imageType=FullColor
     mirroring=NoMirror
     resourceName=None
     resourceObject=null
     tiling=NoTile
     imageSize=<64 64>
     margin=<ltrb 0 0 0 0>
     outlineSettings {
      bUseBrushTransparency=false
      roundingType=HalfHeightRadius
      width=0
      color {colorUseRule=UseColor_Specified, specifiedColor=<rgba 0 0 0 0>}
      cornerRadii=<0 0 0 0>
     }
     tintColor {colorUseRule=UseColor_Specified, specifiedColor=<rgba 1 1 1 1>}
     uVRegion {bIsValid=false, max=<0 0>, min=<0 0>}
    }
    brushColor=<rgba 0 0 0 0>
    padding=<ltrb 4 0 4 2>
   }
  }
  KillFeedEntry_Text {
   widgetClassPath=&CommonRichTextBlock
   bIsVariable=true
   parent=Border_0
   widget=KillFeedEntry_Text
   properties {
    textStyleSet=/Game/Athena/HUD/TextStyles/AthenaKillFeedStyles
    decoratorClasses [/Game/Athena/HUD/TextStyles/KillFeedImageDecorator.KillFeedImageDecorator_C, /Script/FortniteUI.FortRichTextTextureDecorator]
    defaultTextStyleOverride {
     overflowPolicy=Clip
     transformPolicy=None
     colorAndOpacity {colorUseRule=UseColor_Specified, specifiedColor=<rgba 1 1 1 1>}
     font {
      bForceMonospaced=false
      bMaterialIsStencil=false
      fontMaterial=null
      fontObject=&NotoSans
      letterSpacing=0
      monospacedWidth=1
      size=24
      skewAmount=0
      typefaceFontName=NotoSans-Regular
      outlineSettings {
       bApplyOutlineToDropShadows=false
       bMiteredCorners=false
       bSeparateFillAlpha=false
       outlineMaterial=null
       outlineSize=0
       outlineColor=<rgba 0 0 0 1>
      }
     }
     highlightColor {colorUseRule=UseColor_Specified, specifiedColor=<rgba 0 0 0 1>}
     highlightShape {
      bIsDynamicallyLoaded=false
      drawAs=Image
      imageType=NoImage
      mirroring=NoMirror
      resourceName=None
      resourceObject=null
      tiling=NoTile
      imageSize=<32 32>
      margin=<ltrb 0 0 0 0>
      outlineSettings {
       bUseBrushTransparency=false
       roundingType=HalfHeightRadius
       width=0
       color {colorUseRule=UseColor_Specified, specifiedColor=<rgba 0 0 0 0>}
       cornerRadii=<0 0 0 0>
      }
      tintColor {colorUseRule=UseColor_Specified, specifiedColor=<rgba 1 1 1 1>}
      uVRegion {bIsValid=false, max=<0 0>, min=<0 0>}
     }
     selectedBackgroundColor {colorUseRule=UseColor_Foreground, specifiedColor=<rgba 1 0 1 1>}
     shadowColorAndOpacity=<rgba 0 0 0 1>
     shadowOffset=<0 0>
     strikeBrush {
      bIsDynamicallyLoaded=false
      drawAs=Image
      imageType=NoImage
      mirroring=NoMirror
      resourceName=None
      resourceObject=null
      tiling=NoTile
      imageSize=<32 32>
      margin=<ltrb 0 0 0 0>
      outlineSettings {
       bUseBrushTransparency=false
       roundingType=HalfHeightRadius
       width=0
       color {colorUseRule=UseColor_Specified, specifiedColor=<rgba 0 0 0 0>}
       cornerRadii=<0 0 0 0>
      }
      tintColor {colorUseRule=UseColor_Specified, specifiedColor=<rgba 1 1 1 1>}
      uVRegion {bIsValid=false, max=<0 0>, min=<0 0>}
     }
     underlineBrush {
      bIsDynamicallyLoaded=false
      drawAs=Image
      imageType=NoImage
      mirroring=NoMirror
      resourceName=None
      resourceObject=null
      tiling=NoTile
      imageSize=<32 32>
      margin=<ltrb 0 0 0 0>
      outlineSettings {
       bUseBrushTransparency=false
       roundingType=HalfHeightRadius
       width=0
       color {colorUseRule=UseColor_Specified, specifiedColor=<rgba 0 0 0 0>}
       cornerRadii=<0 0 0 0>
      }
      tintColor {colorUseRule=UseColor_Specified, specifiedColor=<rgba 1 1 1 1>}
      uVRegion {bIsValid=false, max=<0 0>, min=<0 0>}
     }
    }
    shapedTextOptions {bOverride_TextFlowDirection=true, bOverride_TextShapingMethod=false, textFlowDirection=Culture, textShapingMethod=Auto}
   }
   slot {class=BorderSlot, properties {padding=<ltrb 4 0 4 2>}}
  }
 ]
}
