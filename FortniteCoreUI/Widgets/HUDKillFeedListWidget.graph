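# HUD kill feed list widget, a prioritized widget that shows HUDKillFeedEntryWidget entries in a Common List View inside a size box.
# On Construct it right-aligns the list if a console variable enables that, then anchors it to the bottom for touch input or the top otherwise.
# CurrentEntriesChanged marks the widget's content active while the entry count is above zero, and inactive otherwise.
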WidgetBlueprint /FortniteCoreUI/Widgets/HUDKillFeedListWidget
graphs [
 CurrentEntriesChanged {nodes [
   n1 = CurrentEntriesChanged!() -> (then, CurrentEntryCount:int) {is=title K2Node_FunctionEntry}
   > n2 = Branch(Condition="integer > integer"(A=@n1))
   > SetHasActiveContent!(execute, self:Fort Prioritized Widget Interface, bNewActive:bool=true) -> (then)
   SetHasActiveContent(execute=@n2.else, bNewActive=false)
  ]}
 EventGraph {nodes [
   Event Construct()
   > n1 = GetConsoleVariableBoolValue!(execute, VariableName:String=BattleRoyalGameUI.KillFeedPositionOnRight.Enabled) -> (then, ReturnValue:bool)
   > n2 = Branch(Condition=@n1)
   > n3 = SetHorizontalAlignment(
    self:Size Box Slot Object Reference=SlotAsSizeBoxSlot!(Widget:Widget Object Reference=Get ListView() -> (ListView:Common List View Object Reference)) -> (ReturnValue:Size Box Slot Object Reference)
    InHorizontalAlignment=HAlign_Right
   )
   n4 = GetOwningPlayer()
   SetVerticalAlignment(
    execute=[@n3.then, @n2.else]
    self:Size Box Slot Object Reference=SlotAsSizeBoxSlot(Widget=Get ListView())
    InVerticalAlignment=Select(
     Option 0:EVerticalAlignment Enum=VAlign_Top
     Option 1:EVerticalAlignment Enum=VAlign_Bottom
     Index=IsUsingInputType!(InputType:ECommonInputType Enum=Touch, PlayerController:Player Controller Object Reference=@n4) -> (ReturnValue:bool)
    ) -> (ReturnValue:EVerticalAlignment Enum)
   )
  ]}
]
variables []
widgets {
 info {inheritedWidgetCount=1, parentClass=/Script/FortniteUI.AthenaCommonPrioritizedWidget, rootWidgetClass=&SizeBox, widgetCount=3}
 widgets [
  SizeBox_0 {
   widgetClassPath=&SizeBox
   widget=SizeBox_0
   properties {
    bOverride_MinDesiredHeight=true
    bOverride_MinDesiredWidth=true
    maxDesiredWidth=600
    minDesiredHeight=150
    minDesiredWidth=700
    widthOverride=600
   }
  }
  ListView {
   widgetClassPath=&CommonListView
   bIsVariable=true
   parent=SizeBox_0
   widget=ListView
   properties {entryWidgetClass=/Game/Athena/HUD/HUDKillFeedEntryWidget.HUDKillFeedEntryWidget_C}
   slot {class=SizeBoxSlot, properties {horizontalAlignment=HAlign_Left, verticalAlignment=VAlign_Top}}
  }
  TutorialHighlight {widgetClassPath=&TutorialHighlight}
 ]
}
