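# Widget Blueprint layout for FortSubtitleWidget that shows centered dialog subtitles at the bottom of the Fortnite HUD.
# An Overlay holds a FortHUDElementWrapper, tagged as center subtitles and hide-for-capture, whose named slot hosts DialogBorder and its wrapping DialogSubtitleBlock text.
# Of the parent's inherited widget bindings, only the dialog pair is placed; caption and description ones are absent.
# The event graph has only unconnected event stubs, leaving behavior to the parent class.
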WidgetBlueprint /FortniteCoreUI/Widgets/WBP_FortSubtitleWidget
graphs [EventGraph {nodes [
   Event PreConstruct()
   Event Construct()
   Event Tick()
  ]}]
variables []
widgets {
 info {
  inheritedWidgetCount=6
  namedSlotCount=1
  parentClass=/Script/SubtitlesWidgets.FortSubtitleWidget
  rootWidgetClass=&Overlay
  widgetCount=8
 }
 widgets [
  Overlay_18 {widgetClassPath=&Overlay, widget=Overlay_18}
  FortHUDElementWrapper {
   widgetClassPath=/Game/Athena/HUD/FortHUDElementWrapper
   bIsVariable=true
   parent=Overlay_18
   widget=FortHUDElementWrapper
   properties {elementTag {tagName=HUD.Athena}, hUDElementTag {gameplayTags [{tagName=Athena.HideForCapture}, {tagName=HUD.Center.Subtitles}]}}
   slot {class=OverlaySlot, properties {horizontalAlignment=HAlign_Center, verticalAlignment=VAlign_Bottom, padding=<ltrb 0 0 0 50>}}
  }
  DialogBorder {
   widgetClassPath=&CommonBorder
   bInherited=true
   namedSlotHost=FortHUDElementWrapper
   widget=DialogBorder
   properties {horizontalAlignment=HAlign_Center, verticalAlignment=VAlign_Center, padding=<ltrb 7 5 7 5>}
  }
  DialogSubtitleBlock {
   widgetClassPath=&CommonTextBlock
   bInherited=true
   parent=DialogBorder
   widget=DialogSubtitleBlock
   properties {
    justification=Center
    text=SEE-Bot: SEE-Bot SEES THE INSIDE OF THE BAG. THERE IS SO MUCH BAG TO SEE.
    wrapTextAt=1170
    font {
     bForceMonospaced=false
     bMaterialIsStencil=false
     fontMaterial=null
     fontObject=&NotoSans
     letterSpacing=0
     monospacedWidth=1
     size=24
     skewAmount=0
     typefaceFontName=NotoSans-Regular
     outlineSettings {
      bApplyOutlineToDropShadows=false
      bMiteredCorners=false
      bSeparateFillAlpha=false
      outlineMaterial=null
      outlineSize=0
      outlineColor=<rgba 0 0 0 1>
     }
    }
    shadowOffset=<0 0>
   }
   slot {class=BorderSlot, properties {horizontalAlignment=HAlign_Center, verticalAlignment=VAlign_Center, padding=<ltrb 7 5 7 5>}}
  }
  CaptionSubtitleBlock {widgetClassPath=&TextBlock, bInherited=true, widget=null, properties=~}
  DescriptionSubtitleBlock {widgetClassPath=&TextBlock, bInherited=true, widget=null, properties=~}
  CaptionBorder {widgetClassPath=&Border, bInherited=true, widget=null, properties=~}
  DescriptionBorder {widgetClassPath=&Border, bInherited=true, widget=null, properties=~}
 ]
}
